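# Material function for the YokedOutNegative wheel cosmetic VFX that converts a 2D vector or centered UVs into radial coordinates.
# Arctangent2 yields the angle, which an If remaps to a 0-1 range by wrapping negative values through a full turn.
# A square root of summed squared components gives the length.
# It outputs the angle alone and an angle/length pair, with a static switch setting the channel order.
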MaterialFunction /VehicleCosmetics/Wheels/YokedOutNegative/VFX/MF_YokedOutNegative_NS
graph {nodes [
  n1 = FunctionInput(Preview=Subtract(A=TextureCoordinate() {uTiling=2, vTiling=2}, B=Constant() {r=1})) {
   bUsePreviewValueAsDefault=true
   description="Insert a vector 2 value to retrieve the vectors angle or insert uvs to retrieve radial coordinates. Note that (0,0) is not an acceptable number.  "
   inputName=Vector or UVs
   inputType=FunctionInput_Vector2
   previewValue=<1 1 0 1>
  }
  n2 = ComponentMask(None=Arctangent2!(Y=ComponentMask(None=@n1) {g=true}, X=ComponentMask(None=@n1) {r=true})) {r=true}
  n3 = Constant() {r=6.2832}
  n4 = Divide(A=@n2, B=@n3)
  n5 = If(A=@n2, B=Constant(), "A > B"=@n4, "A == B"=@n4, "A < B"=Divide(A=Add(A=@n2, B=@n3), B=@n3))
  FunctionOutput(None=@n5) {
   description=This will return the angle of of the input vector or a radial gradient in the case of input uvs.
   outputName=Vector Converted to Angle
   sortPriority=1
  }
  n6 = Multiply(A=@n1, B=@n1)
  n7 = SquareRoot(None=Add(A=ComponentMask(None=@n6) {r=true}, B=ComponentMask(None=@n6) {g=true}))
  n8 = FunctionInput(Preview=StaticBool() {value=true}) {
   bUsePreviewValueAsDefault=true
   inputName="Switch the order of the channels exported from the \"radial coordinates\" output."
   inputType=FunctionInput_StaticBool
  }
  FunctionOutput(None=StaticSwitch(True=AppendVector(A=@n5, B=@n7), False=AppendVector(A=@n7, B=@n5), Value=@n8)) {
   description=The input coordinates or vector will be transformed into radial coordinates. In the case of a vector the angle will be output in one channel and the length of the vector in the other.
   outputName=Radial Coordinates
  }
 ]}
