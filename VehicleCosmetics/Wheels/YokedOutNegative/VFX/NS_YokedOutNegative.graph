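# Niagara effect for the YokedOutNegative vehicle wheel cosmetic, drawing elongated blue sprites that are pulled inward from a ring.
# A single GPU emitter spawns sprites at a steady rate on the surface of a thin, rotated cylinder.
# Point attraction draws them toward an attractor and kills them inside a small radius, while a curve shrinks them over their short lifetime.
# The color-over-life module is disabled, so sprites keep their initial color.
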NiagaraSystem /VehicleCosmetics/Wheels/YokedOutNegative/VFX/NS_YokedOutNegative
dependencies {
 usedDataInterfaces [&NiagaraDataInterfaceColorCurve, &NiagaraDataInterfaceCurve]
 usedDynamicInputs []
 usedModules [&SystemState]
}
emitters {CompletelyEmpty {
  simTarget=GPUComputeSim
  emitterUpdateScript [
   &EmitterState(Life Cycle Mode=::NewEnumerator1, Loop Duration=1, UseLoopDelay=false)
   &SpawnRate(SpawnRate=40, Use Spawn Probability=false)
  ]
  particleSpawnScript [
   &InitializeParticle(
    Color=<rgba 0 0.0802 0.4865 1>
    Lifetime Max=1.2
    Lifetime Min=0.8
    Lifetime Mode=::NewEnumerator1
    Sprite Size=<1.5 20>
    Sprite Size Mode=::NewEnumerator4
    UsePositionOffset=false
    Write Material Random=true
   )
   &ShapeLocation(
    Cylinder Height=5
    Cylinder Height Midpoint=0
    Cylinder Radius=60
    MaskSpawnGroup=false
    Offset Mode=::NewEnumerator0
    Override Seed=false
    Rotation Axis=<90 0 0>
    Rotation Mode=::NewEnumerator1
    Shape Primitive=::NewEnumerator1
    Surface Only=true
    UseOwnerScale=true
   )
  ]
  particleUpdateScript [
   &ParticleState()
   &Color(Color=&ColorFromCurve(ColorCurve=DataInterface() {propertyValues {
       AlphaCurve {keys [<at 0 1 RCIM_Cubic>,<at 1 0 RCIM_Cubic>]}
       BlueCurve {keys [<at 0 1 RCIM_Cubic>,<at 0.5 1 RCIM_Cubic>,<at 1 1 RCIM_Cubic>]}
       GreenCurve {keys [
         <at 0 0.2603 RCIM_Cubic>
         {arriveTangent=-0.0817, interpMode=RCIM_Cubic, leaveTangent=-0.0817, tangentMode=RCTM_User, time=0.5, value=1}
         <at 1 0.0583 RCIM_Cubic>
        ]}
       RedCurve {keys [
         {arriveTangent=-27.2727, interpMode=RCIM_Cubic, time=0, value=0.1048}
         {arriveTangent=-0.1048, interpMode=RCIM_Cubic, leaveTangent=-0.1048, time=0.4, value=20}
         {interpMode=RCIM_Cubic, leaveTangent=85.4545, time=1, value=0}
        ]}
      }})) {enabled=false}
   &PointAttractionForce(Attraction Strength=25, Kill Radius=12, Kill Within Radius=true, Module.Use Falloff=true, UseAttractorOffset=true)
   &SolveForcesAndVelocity(Clamp Velocity=false, Limit Acceleration=false)
   &ScaleSpriteSize(Uniform Curve Sprite Scale=DataInterface() {propertyValues {Curve {keys [{tangentMode=RCTM_User, time=0.035, value=1.3445}, <at 0.5102 0.52>, {tangentMode=RCTM_User, time=1.0228, value=0.0045}]}}})
  ]
  rendererClasses [&NiagaraSpriteRendererProperties]
  renderers [{rendererClass=&NiagaraSpriteRendererProperties}]
 }}
summary {
 systemName=NS_YokedOutNegative
 emitters [{emitterName=CompletelyEmpty, simTarget=GPUComputeSim, rendererClasses [&NiagaraSpriteRendererProperties]}]
 userVariables []
}
